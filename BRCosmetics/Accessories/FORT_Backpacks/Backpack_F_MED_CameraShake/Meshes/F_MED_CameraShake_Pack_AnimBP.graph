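# Animation Blueprint for the CameraShake backpack cosmetic that continuously spins its "ball" bone.
# Each update eases a spin rate toward a fast value when falling, jumping or moving, and a slow one otherwise.
# That rate times delta time accumulates into a rotator's pitch and yaw, leaving roll unchanged.
# The AnimGraph applies the rotator additively in bone space over the input pose.
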AnimBlueprint /BRCosmetics/Accessories/FORT_Backpacks/Backpack_F_MED_CameraShake/Meshes/F_MED_CameraShake_Pack_AnimBP
target_skeleton=&/F_MED_CameraShake_Pack_Skeleton
graphs [
 AnimGraph {nodes [
   n1 = "Transform (Modify) Bone - Bone: ball"(
    ComponentPose=Local To Component(LocalPose=Input Pose())
    Alpha=1
    Translation="0,0,0"
    Rotation=Get rotator() -> (rotator:Rotator)
    Scale="1,1,1"
   ) {
    rotationMode=BMM_Additive
    rotationSpace=BCS_BoneSpace
    translationMode=BMM_Additive
    translationSpace=BCS_BoneSpace
    alphaBoolBlend {blendOption=Linear}
    boneToModify {boneName=ball}
   }
   Output Pose(Result=Component To Local(ComponentPose=@n1)) {layerGroup=None}
  ]}
 EventGraph {nodes [
   n1 = Event BlueprintUpdateAnimation()
   > Set delta time x(delta time x:float=@n1.DeltaTimeX) -> (Output_Get:float)
   n2 = OR Boolean!(A=Get bIsFalling(), B=Get bIsJumping(), C:bool=Get bIsMoving2D()) -> (ReturnValue)
   > Set multiplier(multiplier:double=FInterpTo(
     Current=Get multiplier() -> (multiplier:double)
     Target=SelectFloat(A=-300, B=-15, bPickA=@n2)
     DeltaTime=Get delta time x() -> (delta time x:float)
     InterpSpeed=2
    )) -> (Output_Get:double)
   n3 = BreakRotator(InRot=Get rotator())
   n4 = float * float(A=Get multiplier(), B=Get delta time x()) {is=title K2Node_PromotableOperator}
   > Set rotator(rotator:Rotator=MakeRotator(Roll=@n3, Pitch=float + float(A=@n3.Pitch, B=@n4), Yaw=float + float(A=@n3.Yaw, B=@n4))) -> (Output_Get:Rotator)
   TryGetPawnOwner()
  ]}
]
variables [
 delta time x {type=number}
 multiplier {type=number}
 rotator {type=Rotator}
]
